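# Animation Blueprint for the female medium NoiseLess player body mesh, copying the pose from its attached parent mesh and adding secondary zipper motion.
# The copied pose runs through the body's Control Rig, then constraints drive the left and right zipper master bones from pelvis, thigh and thigh twist bones using uneven weights.
# Root-relative AnimDynamics then simulates each zipper bone with gravity, angular springs and angular limits.
# The event graph holds no logic.
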AnimBlueprint /Game/Characters/Player/Female/Medium/Bodies/F_MED_NoiseLess/Meshes/F_MED_NoiseLess_AnimBP
target_skeleton=&/F_MED_NoiseLess_Skeleton
graphs [
 AnimGraph {nodes [
   n1 = Control Rig(Source=Copy Pose From Mesh() {bUseAttachedParent=true}) {
    alphaBoolBlend {blendOption=Linear}
    controlRigAssetReference {blueprintRigClass=&/F_MED_NoiseLess_CtrlRig.F_MED_NoiseLess_CtrlRig_C}
   }
   n2 = "Constraint - dyn_zipper_l_MSTR (PPP)"!(
    ComponentPose=Local To Component(LocalPose=@n1)
    Alpha=1
    ConstraintWeights_0=0.037
    ConstraintWeights_1:float=0.5
    ConstraintWeights_2:float=0.463
   ) -> (Pose) {
    alphaBoolBlend {blendOption=Linear}
    boneToModify {boneName=dyn_zipper_l_MSTR}
    constraintSetup [{targetBone {boneName=pelvis}}, {targetBone {boneName=thigh_l}}, {targetBone {boneName=thigh_twist_01_l}}]
    constraintWeights [1,1,1]
   }
   n3 = "Constraint - dyn_zipper_r_MSTR (PPP)"!(ComponentPose=@n2, Alpha=1, ConstraintWeights_0=0.037, ConstraintWeights_1:float=0.5, ConstraintWeights_2:float=0.463) -> (Pose) {
    alphaBoolBlend {blendOption=Linear}
    boneToModify {boneName=dyn_zipper_r_MSTR}
    constraintSetup [{targetBone {boneName=pelvis}}, {targetBone {boneName=thigh_r}}, {targetBone {boneName=thigh_twist_01_r}}]
    constraintWeights [1,1,1]
   }
   n4 = Anim Dynamics - Bone: dyn_zipper_l!(ComponentPose=@n3, Alpha=1, GravityOverride:Vector="0,0,-980", ExternalForce="0,0,0") -> (Pose) {
    angularBiasOverride=1
    angularDampingOverride=0.999
    angularSpringConstant=1000
    bAngularSpring=true
    bOverrideAngularDamping=true
    bUseGravityOverride=true
    bUsePlanarLimit=false
    simulationSpace=RootRelative
    alphaBoolBlend {blendInTime=0.5, blendOption=Linear, blendOutTime=0.7}
    boundBone {boneName=dyn_zipper_l}
    componentAppliedLinearAccClamp=<3000 3000 3000>
    componentLinearAccScale=<0.3 0.3 0.3>
    componentLinearVelScale=<0.4 0.4 0.4>
    physicsBodyDefinitions [{
      collisionType=CustomSphere
      sphereCollisionRadius=2.5
      boundBone {boneName=dyn_zipper_l}
      boxExtents=<4 4 4>
      constraintSetup {angularLimitsMax=<0 0 80>, angularLimitsMin=<0 1 0>, angularTarget=<0 0 0>}
      localJointOffset=<2.5 0 0>
     }]
   }
   n5 = Anim Dynamics - Bone: dyn_zipper_r!(ComponentPose=@n4, Alpha=1, GravityOverride:Vector="0,0,-980", ExternalForce="0,0,0") -> (Pose) {
    angularBiasOverride=1
    angularDampingOverride=0.999
    angularSpringConstant=1000
    bAngularSpring=true
    bOverrideAngularDamping=true
    bUseGravityOverride=true
    bUsePlanarLimit=false
    simulationSpace=RootRelative
    alphaBoolBlend {blendInTime=0.5, blendOption=Linear, blendOutTime=0.7}
    boundBone {boneName=dyn_zipper_r}
    componentAppliedLinearAccClamp=<3000 3000 3000>
    componentLinearAccScale=<0.3 0.3 0.3>
    componentLinearVelScale=<0.4 0.4 0.4>
    physicsBodyDefinitions [{
      collisionType=CustomSphere
      sphereCollisionRadius=2.5
      boundBone {boneName=dyn_zipper_r}
      boxExtents=<4 4 4>
      constraintSetup {angularLimitsMax=<0 0 90>, angularLimitsMin=<0 0 0>, angularTarget=<0 0 0>}
      localJointOffset=<2.5 0 0>
     }]
   }
   Output Pose(Result=Component To Local(ComponentPose=@n5)) {layerGroup=None}
  ]}
 EventGraph {nodes [Event BlueprintUpdateAnimation()]}
]
variables []
